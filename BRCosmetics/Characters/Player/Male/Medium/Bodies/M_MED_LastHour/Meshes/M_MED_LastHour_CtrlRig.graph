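# Control Rig for the male medium LastHour body, adding deformation and pose correctives for jacket, collar, katana holder and thigh trinket.
# It runs shared DeformRig, then caches thigh spherical pose readers, pelvis height, calf yaw and a DBNO alpha.
# When RBAN is disabled, it only constrains the katana holder to the right thigh.
# Otherwise, those values, spine bend and an arm reader weight additive bone offsets, some Kalman-smoothed.
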ControlRigBlueprint /BRCosmetics/Characters/Player/Male/Medium/Bodies/M_MED_LastHour/Meshes/M_MED_LastHour_CtrlRig
graphs [
 RigVMFunctionLibrary {links [], nodes []}
 RigVMFunction_Sequence_1_ContainedGraph {nodes [
   n1 = RigVMFunctionEntryNode()
   RigVMFunction_Sequence = &RigVMFunction_Sequence(ExecuteContext=@n1.ExecuteContext)
   n2 = &RigVMFunction_Sequence(ExecuteContext=@RigVMFunction_Sequence.B)
   RigVMFunctionReturnNode(A=@RigVMFunction_Sequence.A, B=@n2.A, C=@n2.B)
  ]}
 RigVMModel {nodes [
   RigUnit_BeginExecution = &RigUnit_BeginExecution()
   FN_DeformRig_FNC = RigVMFunctionReferenceNode(
    DeformRig_ Curve="(deformRig_l_upperarm_up_alpha_72_#1=False,deformRig_r_upperarm_up_alpha_73_#2=False,deformRig_l_upperarm_fr_alpha_74_#3=False,deformRig_r_upperarm_fr_alpha_75_CB51B8 ...(truncated)"
    DeformRig_Bool="(upperarm_1_#4=True,elbow_3_#5=True,wrist_7_#6=True,groin_8_#7=True,glute_17_#8=True,knee_9_40C408324CA261 ...(truncated)"
    DeformRig_Debug=(upperarm_debug_18_#4=False,wrist_debug_20_#6=False,lat_debug_26_#9=False,groinGlute_debug_27_#7=False)
    DeformRig_Value="(upperarm_up_18_#10=(Rotation=(X=0,Y=0,Z=0,W=1),Translation=(X=0,Y=0,Z=8),Scale3D=(X=0.7,Y=1,Z=1)),upperarm_fwd_fr_19_#11=(Rotation=(X=0,Y=0,Z=0,W=1),Translation=(X=0,Y=10,Z=0),Scale3D=(X=0.5,Y=1,Z= ...(truncated)"
    ExecuteContext=@RigUnit_BeginExecution.ExecutePin
   )
   > n1 = &RigUnit_SphericalPoseReader(
    ActiveRegionScaleFactors=(PositiveWidth=0.1,NegativeWidth=1,PositiveHeight=1,NegativeHeight=1)
    ActiveRegionSize=0.3
    Debug=(bDrawDebug=true,bDraw2D=False,bDrawLocalAxes=False,DebugScale=25,DebugSegments=20,DebugThickness=0.25)
    DriverAxis=(X=1,Y=0,Z=0)
    DriverItem=(Type=Bone,Name="thigh_r")
    FalloffRegionScaleFactors=(PositiveWidth=0.7,NegativeWidth=1,PositiveHeight=1,NegativeHeight=1)
    FalloffSize=0.5
    FlipHeightScaling=False
    FlipWidthScaling=False
    OptionalParentItem=(Type=Bone,Name="pelvis")
    RotationOffset=(X=-90,Y=-7,Z=0)
   )
   > RigVMVariableNode(Value=@n1.OutputParam) {variable=SPR_Thing_R}
   > n2 = &RigUnit_SphericalPoseReader(
    ActiveRegionScaleFactors=(PositiveWidth=1,NegativeWidth=0.1,PositiveHeight=1,NegativeHeight=1)
    ActiveRegionSize=0.3
    Debug=(bDrawDebug=true,bDraw2D=False,bDrawLocalAxes=False,DebugScale=25,DebugSegments=20,DebugThickness=0.25)
    DriverAxis=(X=-1,Y=0,Z=0)
    DriverItem=(Type=Bone,Name="thigh_l")
    FalloffRegionScaleFactors=(PositiveWidth=1,NegativeWidth=0.7,PositiveHeight=1,NegativeHeight=1)
    FalloffSize=0.5
    FlipHeightScaling=False
    FlipWidthScaling=False
    OptionalParentItem=(Type=Bone,Name="pelvis")
    RotationOffset=(X=90,Y=-7,Z=0)
   )
   > RigVMVariableNode(Value=@n2.OutputParam) {variable=SPR_Thing_L}
   > n3 = &RigUnit_SphericalPoseReader(
    ActiveRegionScaleFactors=(PositiveWidth=0.1,NegativeWidth=1,PositiveHeight=1,NegativeHeight=1)
    ActiveRegionSize=0.3
    Debug=(bDrawDebug=false,bDraw2D=False,bDrawLocalAxes=False,DebugScale=25,DebugSegments=20,DebugThickness=0.25)
    DriverAxis=(X=1,Y=0,Z=0)
    DriverItem=(Type=Bone,Name="thigh_r")
    FalloffRegionScaleFactors=(PositiveWidth=1,NegativeWidth=0.7,PositiveHeight=1,NegativeHeight=1)
    FalloffSize=0.5
    FlipHeightScaling=False
    FlipWidthScaling=False
    OptionalParentItem=(Type=Bone,Name="pelvis")
    RotationOffset=(X=0,Y=15,Z=0)
   )
   > RigVMVariableNode(Value=@n3.OutputParam) {variable=SPR_Thing_R_Side}
   n4 = RigVMVariableNode() {variable=SPR_Thing_R}
   n5 = RigVMVariableNode() {variable=SPR_Thing_L}
   RigVMFunction_MathBoolToFloat = &RigVMFunction_MathBoolToFloat(Value=RigVMVariableNode() {variable=Is dbno}.Value)
   > RigVMVariableNode(Value=&RigVMFunction_MathDoubleMul(A=&RigVMFunction_MathDoubleMul(A=@n4.Value, B=@n5.Value).Result, B=@RigVMFunction_MathBoolToFloat.Result).Result) {variable=Is_Dbno_alpha}
   > RigVMVariableNode(Value=&RigUnit_GetRelativeTransformForItem(Child=(Type=Bone,Name="pelvis"), Parent=(Type=Bone,Name="root"), bChildInitial=False, bParentInitial=False).RelativeTransform.Translation.Z) {variable=Get pelvis from root by X}
   > Get Delta Ref = RigVMFunctionReferenceNode(Item=(Type=Bone,Name="calf_r"))
   > RigVMVariableNode(Value=@"Get Delta Ref".Result.Rotation.Yaw) {variable=Delta_Calf_R}
   > Get Delta Ref_1 = RigVMFunctionReferenceNode(Item=(Type=Bone,Name="calf_l"))
   > n6 = RigVMVariableNode(Value=@"Get Delta Ref_1".Result.Rotation.Yaw) {variable=Delta_Calf_L}
   RigVMFunction_ControlFlowBranch = &RigVMFunction_ControlFlowBranch(Condition=RigVMVariableNode() {variable=RBAN is disable}.Value, ExecuteContext=@n6.ExecuteContext)
   &RigUnit_ParentConstraint(
    AdvancedSettings=(InterpolationType=Average,RotationOrderForFilter=XZY)
    Child=(Type=Bone,Name="dyn_katana_holder_02")
    Parents=((Item=(Type=Bone,Name="thigh_r"),Weight=1))
    Weight=1
    ExecutePin=@RigVMFunction_ControlFlowBranch.True
   )
   RigVMFunction_Sequence = RigVMAggregateNode(ExecuteContext=@RigVMFunction_ControlFlowBranch.False)
   KalmanFilter = &RigVMFunction_KalmanFloat(BufferSize=16, Value=RigVMVariableNode() {variable=SPR_Thing_L}.Value) {template="KalmanFilter::Execute(in Value,in BufferSize,out Result)"}
   RigVMFunction_MathQuaternionToEuler = &RigVMFunction_MathQuaternionToEuler(
    RotationOrder=ZYX
    Value=&RigUnit_GetRelativeTransformForItem(Child=(Type=Bone,Name="spine_05"), Parent=(Type=Bone,Name="pelvis"), bChildInitial=False, bParentInitial=False).RelativeTransform.Rotation
   )
   Remap = &RigVMFunction_MathDoubleRemap(SourceMaximum=10, SourceMinimum=0, TargetMaximum=1, TargetMinimum=0, Value=@RigVMFunction_MathQuaternionToEuler.Result.Z)
   Multiply = &RigVMFunction_MathDoubleMul(B=@KalmanFilter.Result, A=@Remap.Result)
   &RigUnit_ModifyTransforms(
    ItemToModify="((Item=(Type=Bone,Name=\"dyn_jacket_fr_l_base\"),Transform=(Rotation=(X=-0.067,Y=-0.25,Z=-0.25,W=0.933),Translation=(X=0,Y=0,Z=0),Scale3D=(X=1,Y=1,Z=1))),(Item=(Type=Bone,Name=\"dyn_jacket_s_l_base\"),Transform=(Rotation=(X=0.0885,Y=-0.3304,Z=0.2432,W=0.9077), ...(truncated)"
    WeightMinimum=0
    Weight=@Multiply.Result
    ExecutePin=@RigVMFunction_Sequence.A
   )
   > &RigUnit_ModifyTransforms(
    ItemToModify="((Item=(Type=Bone,Name=\"dyn_jacket_b_l_base\"),Transform=(Rotation=(X=0,Y=-0.342,Z=0,W=0.9397),Translation=(X=-4,Y=0,Z=0),Scale3D=(X=1,Y=1,Z=1))),(Item=(Type=Bone,Name=\"dyn_jacket_b_r_base\"),Transform=(Rotation=(X=0,Y=-0.4226,Z=0,W=0.9063),Translation=(X=-4 ...(truncated)"
    WeightMinimum=0
    Weight=&RigVMFunction_MathDoubleRemap(SourceMaximum=-80, SourceMinimum=-50, TargetMaximum=1, TargetMinimum=0, Value=@RigVMFunction_MathQuaternionToEuler.Result.Z).Result
   )
   RigVMFunction_AnimEvalRichCurve = &RigVMFunction_AnimEvalRichCurve(
    Curve="(EditorCurveData=(Keys=((),(InterpMode=RCIM_Cubic,TangentMode=RCTM_User,Time=0.1914,Value=0.4183,ArriveTangent=1.7005,LeaveTangent=1.7005),(InterpMode=RCIM_Cubic,TangentMode=RCTM_User,Time=0.554,Value=0.8216,ArriveTangent=0.4644,LeaveTangent=0.4644),(Time= ...(truncated)"
    SourceMaximum=-90
    SourceMinimum=-20
    TargetMinimum=0
    Value=&RigVMFunction_MathQuaternionToEuler(
     RotationOrder=ZYX
     Value=&RigUnit_GetRelativeTransformForItem(Child=(Type=Bone,Name="spine_05"), Parent=(Type=Bone,Name="pelvis"), bChildInitial=False, bParentInitial=False).RelativeTransform.Rotation
    ).Result.Z
   )
   > &RigUnit_ModifyTransforms(
    ItemToModify="((Item=(Type=Bone,Name=\"dyn_jacket_b_l_base\"),Transform=(Rotation=(X=0,Y=-0.342,Z=0,W=0.9397),Translation=(X=0,Y=0,Z=0),Scale3D=(X=1,Y=1,Z=1))),(Item=(Type=Bone,Name=\"dyn_jacket_b_r_base\"),Transform=(Rotation=(X=0,Y=-0.342,Z=0,W=0.9397),Translation=(X=0,Y= ...(truncated)"
    WeightMinimum=0
    Weight=@RigVMFunction_AnimEvalRichCurve.Result
   )
   > &RigUnit_ModifyTransforms(
    ItemToModify="((Item=(Type=Bone,Name=\"dyn_jacket_fr_r_base\"),Transform=(Rotation=(X=-0.25,Y=-0.25,Z=-0.067,W=0.933),Translation=(X=0,Y=0,Z=1.5),Scale3D=(X=1,Y=1,Z=1))),(Item=(Type=Bone,Name=\"dyn_katana_holder_base\"),Transform=(Rotation=(X=-0.1736,Y=0,Z=0,W=0.9848),Trans ...(truncated)"
    WeightMinimum=0
    Weight=&RigVMFunction_KalmanFloat(BufferSize=16, Value=RigVMVariableNode() {variable=SPR_Thing_R}.Value) {template="KalmanFilter::Execute(in Value,in BufferSize,out Result)"}.Result
   )
   > &RigUnit_ModifyTransforms(
    ItemToModify="((Item=(Type=Bone,Name=\"dyn_jacket_fr_l_base\"),Transform=(Rotation=(X=0.1786,Y=-0.383,Z=0.383,W=0.8214),Translation=(X=0,Y=0,Z=0),Scale3D=(X=1,Y=1,Z=1))),(Item=(Type=Bone,Name=\"dyn_jacket_s_l_base\"),Transform=(Rotation=(X=0,Y=0.2588,Z=0,W=0.9659),Translati ...(truncated)"
    WeightMinimum=0
    Weight=RigVMVariableNode() {variable=Is_Dbno_alpha}.Value
   )
   n7 = &RigVMFunction_MathDoubleRemap(SourceMaximum=13, SourceMinimum=20, TargetMaximum=1, TargetMinimum=0, Value=RigVMVariableNode() {variable=Get pelvis from root by X}.Value)
   > &RigUnit_ModifyTransforms(
    ItemToModify="((Item=(Type=Bone,Name=\"dyn_jacket_fr_l_base\"),Transform=(Rotation=(X=0,Y=-0.5,Z=0,W=0.866),Translation=(X=-4,Y=0,Z=2),Scale3D=(X=1,Y=1,Z=1))),(Item=(Type=Bone,Name=\"dyn_jacket_s_l_base\"),Transform=(Rotation=(X=0,Y=0,Z=0.342,W=0.9397),Translation=(X=0,Y=0, ...(truncated)"
    WeightMinimum=0
    Weight=@n7.Result
   )
   > &RigUnit_ModifyTransforms(
    ItemToModify="((Item=(Type=Bone,Name=\"dyn_katana_holder_base\"),Transform=(Rotation=(X=0.067,Y=0.25,Z=-0.25,W=0.933),Translation=(X=0,Y=-2,Z=0),Scale3D=(X=1,Y=1,Z=1))),(Item=(Type=Bone,Name=\"dyn_jacket_fr_r_base\"),Transform=(Rotation=(X=-0.0885,Y=-0.3304,Z=-0.2432,W=0.90 ...(truncated)"
    WeightMinimum=0
    Weight=&RigVMFunction_MathDoubleMul(A=@n7.Result, B=RigVMVariableNode() {variable=SPR_Thing_R_Side}.Value).Result
   )
   n8 = &RigUnit_SphericalPoseReader(
    ActiveRegionScaleFactors=(PositiveWidth=1,NegativeWidth=1,PositiveHeight=1,NegativeHeight=1)
    ActiveRegionSize=0.1
    Debug=(bDrawDebug=True,bDraw2D=False,bDrawLocalAxes=False,DebugScale=25,DebugSegments=20,DebugThickness=0.25)
    DriverAxis=(X=-1,Y=0,Z=0)
    DriverItem=(Type=Bone,Name="upperarm_r")
    FalloffRegionScaleFactors=(PositiveWidth=1,NegativeWidth=1,PositiveHeight=1,NegativeHeight=1)
    FalloffSize=0.2
    FlipHeightScaling=False
    FlipWidthScaling=False
    OptionalParentItem=(Type=Bone,Name="spine_05")
    RotationOffset=(X=-100,Y=0,Z=50)
    ExecutePin=@RigVMFunction_Sequence.B
   )
   n9 = RigVMVariableNode() {variable=Get pelvis from root by X}
   > &RigUnit_ModifyTransforms(
    ItemToModify=((Item=(Type=Bone,Name="corrective_collar_r"),Transform=(Rotation=(X=0,Y=0,Z=-0.3827,W=0.9239),Translation=(X=0.9,Y=0,Z=0),Scale3D=(X=1,Y=1,Z=1))))
    WeightMinimum=0
    Weight=&RigVMFunction_MathDoubleMul(
     B=&RigVMFunction_MathDoubleRemap(SourceMaximum=40, SourceMinimum=60, TargetMaximum=1, TargetMinimum=0, Value=@n9.Value).Result
     A=@n8.OutputParam
    ).Result
   )
   &RigUnit_ModifyTransforms(
    ItemToModify=((Item=(Type=Bone,Name="dyn_thigh_trinket_base"),Transform=(Rotation=(X=0.1645,Y=0.5305,Z=0.0454,W=0.8303),Translation=(X=0,Y=0.5,Z=-3),Scale3D=(X=1,Y=1,Z=1))))
    WeightMinimum=0
    ExecutePin=@RigVMFunction_Sequence.C
    Weight=RigVMVariableNode() {variable=SPR_Thing_L}.Value
   )
   &RigUnit_ModifyTransforms(
    ExecutePin=()
    ItemToModify=((Item=(Type=Bone,Name="dyn_jacket_fr_r_base"),Transform=(Rotation=(X=0,Y=0,Z=0,W=1),Translation=(X=0,Y=0,Z=0),Scale3D=(X=1,Y=1,Z=1))))
    Weight=1
    WeightMinimum=0
   )
   &RigUnit_ModifyTransforms(
    ExecutePin=()
    ItemToModify=((Item=(Type=Bone,Name="dyn_katana_holder_base"),Transform=(Rotation=(X=0,Y=0.2588,Z=0,W=0.9659),Translation=(X=-5,Y=0,Z=0),Scale3D=(X=1,Y=1,Z=1))))
    WeightMinimum=0
    Weight=RigVMVariableNode() {variable=SPR_Thing_R_Side}.Value
   )
   RigVMUnitNode(SourceMaximum=(), SourceMinimum=(), TargetMaximum=(), TargetMinimum=(), Value=(), bClamp=False) {template="Remap::Execute(in Value,in SourceMinimum,in SourceMaximum,in TargetMinimum,in TargetMaximum,in bClamp,out Result)"}
   &RigVMFunction_MathDoubleRemap(
    SourceMaximum=1
    SourceMinimum=0
    TargetMaximum=1
    TargetMinimum=0
    bClamp=False
    Value=&RigVMFunction_MathQuaternionToEuler(
     RotationOrder=ZYX
     Value=&RigUnit_GetRelativeTransformForItem(Child=(Type=Bone,Name="spine_05"), Parent=(Type=Bone,Name="pelvis"), bChildInitial=False, bParentInitial=False).RelativeTransform.Rotation
    ).Result.Z
   )
   RigVMVariableNode() {variable=SPR_Thing_R}
  ]}
]
hierarchy [
 root {loc=<0 0 0>, rot=<0 0 0 1>}
 pelvis {parent=root, loc=<0 -0.7071 93.1191>, rot=<0.0127 0.707 -0.0127 -0.707>}
 spine_01 {parent=pelvis, loc=<6.7494 -0.6388 0>, rot=<0 0 0.034 -0.9994>}
 spine_02 {parent=spine_01, loc=<7.252 -0.035 0>, rot=<0 0 0.0238 -0.9997>}
 spine_03 {parent=spine_02, loc=<6.0106 0 0>, rot=<0 0 0.0149 -0.9999>}
 spine_04 {parent=spine_03, loc=<8.1023 0 0>, rot=<0 0 -0.0132 -0.9999>}
 spine_05 {parent=spine_04, loc=<10.0142 0 0>, rot=<0 0 -0.0015 -1>}
 attach_backpack {parent=spine_05, loc=<-0.123 -13.473 0>, rot=<-0.0093 -0.707 0.0093 -0.707>}
 neck_01 {parent=spine_05, loc=<11.3138 0.1529 0>, rot=<0 0 -0.2003 -0.9797>}
 neck_02 {parent=neck_01, loc=<4.5627 0 0>, rot=<0 0 0.0504 -0.9987>}
 head {parent=neck_02, loc=<5.3914 0 0>, rot=<0 0 0.115 -0.9934>}
 clavicle_l {parent=spine_05, loc=<4.3575 -0.1466 -2.3223>, rot=<0.0465 -0.7289 -0.009 -0.6829>}
 upperarm_l {parent=clavicle_l, loc=<15.5111 0 0>, rot=<0.0226 -0.3234 -0.0088 -0.9459>}
 lowerarm_l {parent=upperarm_l, loc=<26.6514 0 0>, rot=<0.0391 -0.0099 -0.2453 -0.9686>}
 hand_l {parent=lowerarm_l, loc=<22.3491 0 0>, rot=<-0.7081 -0.0507 0.0975 -0.6975>}
 hand_attach_l {parent=hand_l, loc=<8.5 -3 0>, rot=<0.7071 0.7071 0 0>}
 thumb_01_l {parent=hand_l, loc=<1.3118 -0.3191 -2.459>, rot=<0.6546 -0.4042 0.0664 -0.6354>}
 thumb_02_l {parent=thumb_01_l, loc=<5.9336 -0.0448 0.0401>, rot=<-0.0037 0.0213 0.298 -0.9543>}
 thumb_03_l {parent=thumb_02_l, loc=<2.2223 0.3682 -0.0907>, rot=<0.0118 -0.013 0.0066 -0.9998>}
 index_metacarpal_l {parent=hand_l, loc=<1.6535 0.0183 -2.96>, rot=<0.0158 -0.0124 -0.0173 -0.9996>}
 index_01_l {parent=index_metacarpal_l, loc=<8.3008 0.0357 -0.0524>, rot=<-0.0044 0.0485 0.0924 -0.9945>}
 index_02_l {parent=index_01_l, loc=<3.3587 -0.3004 0.0452>, rot=<-0.002 0.0027 0.3225 -0.9466>}
 index_03_l {parent=index_02_l, loc=<2.5578 -0.123 -0.0252>, rot=<-0.002 0.0067 0.2558 -0.9667>}
 middle_metacarpal_l {parent=hand_l, loc=<2.1856 -0.1725 -0.5984>, rot=<0.006 0.0168 -0.0329 -0.9993>}
 middle_01_l {parent=middle_metacarpal_l, loc=<7.3275 -0.2447 0.0129>, rot=<-0.0208 0.0042 0.0858 -0.9961>}
 middle_02_l {parent=middle_01_l, loc=<4.0123 0.1401 0.1175>, rot=<-0.0218 0.0258 0.3512 -0.9357>}
 middle_03_l {parent=middle_02_l, loc=<2.6077 0.0534 0.1373>, rot=<0.0082 0.0188 0.1708 -0.9851>}
 ring_metacarpal_l {parent=hand_l, loc=<2.2074 -0.3625 1.3638>, rot=<-0.0284 0.0722 -0.0458 -0.9959>}
 ring_01_l {parent=ring_metacarpal_l, loc=<7.0295 -0.6197 -0.1446>, rot=<-0.0489 -0.0252 0.1892 -0.9804>}
 ring_02_l {parent=ring_01_l, loc=<3.6174 0.4647 -0.1027>, rot=<0 0 0.2806 -0.9598>}
 ring_03_l {parent=ring_02_l, loc=<2.4152 0.114 0.0199>, rot=<0 0 0.1189 -0.9929>}
 pinky_metacarpal_l {parent=hand_l, loc=<1.7616 -0.9753 3.3814>, rot=<-0.0397 0.0814 -0.069 -0.9935>}
 pinky_01_l {parent=pinky_metacarpal_l, loc=<7.1709 -0.4831 -0.1094>, rot=<-0.0654 -0.0035 0.2242 -0.9723>}
 pinky_02_l {parent=pinky_01_l, loc=<2.8274 0.0893 -0.1043>, rot=<0.0037 0.0011 0.2854 -0.9584>}
 pinky_03_l {parent=pinky_02_l, loc=<2.2046 0.0307 -0.0078>, rot=<-0.0002 0.0098 0.1445 -0.9895>}
 weapon_l {parent=hand_l, loc=<-0.1974 -3.4111 0.3811>, rot=<0 0 0 1>}
 deform_wrist_dn_base_l {parent=hand_l, loc=<0 -2.2342 0>, rot=<0 0 0 -1>}
 deform_wrist_dn_l {parent=deform_wrist_dn_base_l, loc=<0 0 0>, rot=<0 0 0 1>}
 deform_wrist_up_base_l {parent=hand_l, loc=<0 2.2682 0>, rot=<0 0 0 -1>}
 deform_wrist_up_l {parent=deform_wrist_up_base_l, loc=<0 0 0>, rot=<0 0 0 1>}
 lowerarm_twist_02_l {parent=lowerarm_l, loc=<9.6199 0 0>, rot=<0 0 0 -1>}
 lowerarm_twist_01_l {parent=lowerarm_l, loc=<17.8549 0 0>, rot=<0 0 0 -1>}
 deform_elbow_in_base_l {parent=lowerarm_l, loc=<0.2112 5.259 0.8274>, rot=<0 0 0 -1>}
 deform_elbow_in_l {parent=deform_elbow_in_base_l, loc=<0 0 0>, rot=<0 0 0 1>}
 deform_elbow_out_base_l {parent=lowerarm_l, loc=<0 -4 0>, rot=<0 0 0 -1>}
 deform_elbow_out_l {parent=deform_elbow_out_base_l, loc=<0 0 0>, rot=<0 0 0 1>}
 upperarm_twist_01_l {parent=upperarm_l, loc=<0.0041 0 0>, rot=<0 0 0 -1>}
 deform_upperarm_bk_base_l {parent=upperarm_twist_01_l, loc=<1.0813 -7.2193 -1.9453>, rot=<0 0 0 1>}
 deform_upperarm_bk_l {parent=deform_upperarm_bk_base_l, loc=<0 0 0>, rot=<0 0 0 1>}
 deform_upperarm_fr_base_l {parent=upperarm_twist_01_l, loc=<-0.3691 7.0506 -2.765>, rot=<0 0 0 1>}
 deform_upperarm_fr_l {parent=deform_upperarm_fr_base_l, loc=<0 0 0>, rot=<0 0 0 1>}
 deform_upperarm_up_base_l {parent=upperarm_twist_01_l, loc=<-3.2216 -1.1391 3.3311>, rot=<0 0 0 1>}
 deform_upperarm_up_l {parent=deform_upperarm_up_base_l, loc=<0 0 0>, rot=<0 0 0 1>}
 upperarm_twist_02_l {parent=upperarm_l, loc=<13.7941 0 0>, rot=<0 0 0 -1>}
 attach_cape {parent=spine_05, loc=<9.017 -11.714 0>, rot=<0.0616 -0.7044 -0.0616 -0.7044>}
 clavicle_r {parent=spine_05, loc=<4.3575 -0.1466 2.3223>, rot=<0.7289 0.0465 -0.6829 0.009>}
 upperarm_r {parent=clavicle_r, loc=<-15.5111 0 0>, rot=<0.0226 -0.3234 -0.0088 -0.9459>}
 lowerarm_r {parent=upperarm_r, loc=<-26.6514 0 0>, rot=<0.0391 -0.0099 -0.2453 -0.9686>}
 hand_r {parent=lowerarm_r, loc=<-22.3491 0 0>, rot=<-0.7081 -0.0507 0.0975 -0.6975>}
 thumb_01_r {parent=hand_r, loc=<-1.3118 0.3191 2.459>, rot=<0.6546 -0.4042 0.0664 -0.6354>}
 thumb_02_r {parent=thumb_01_r, loc=<-5.9336 0.0448 -0.0401>, rot=<-0.0037 0.0213 0.298 -0.9543>}
 thumb_03_r {parent=thumb_02_r, loc=<-2.2223 -0.3682 0.0907>, rot=<0.0118 -0.013 0.0066 -0.9998>}
 index_metacarpal_r {parent=hand_r, loc=<-1.6535 -0.0183 2.96>, rot=<0.0158 -0.0124 -0.0173 -0.9996>}
 index_01_r {parent=index_metacarpal_r, loc=<-8.3008 -0.0357 0.0524>, rot=<-0.0044 0.0485 0.0924 -0.9945>}
 index_02_r {parent=index_01_r, loc=<-3.3587 0.3004 -0.0452>, rot=<-0.002 0.0027 0.3225 -0.9466>}
 index_03_r {parent=index_02_r, loc=<-2.5578 0.123 0.0252>, rot=<-0.002 0.0067 0.2558 -0.9667>}
 middle_metacarpal_r {parent=hand_r, loc=<-2.1856 0.1725 0.5984>, rot=<0.006 0.0168 -0.0329 -0.9993>}
 middle_01_r {parent=middle_metacarpal_r, loc=<-7.3275 0.2447 -0.0129>, rot=<-0.0208 0.0042 0.0858 -0.9961>}
 middle_02_r {parent=middle_01_r, loc=<-4.0123 -0.1401 -0.1175>, rot=<-0.0218 0.0258 0.3512 -0.9357>}
 middle_03_r {parent=middle_02_r, loc=<-2.6077 -0.0534 -0.1373>, rot=<0.0082 0.0188 0.1708 -0.9851>}
 ring_metacarpal_r {parent=hand_r, loc=<-2.2074 0.3625 -1.3638>, rot=<-0.0284 0.0722 -0.0458 -0.9959>}
 ring_01_r {parent=ring_metacarpal_r, loc=<-7.0295 0.6197 0.1446>, rot=<-0.0489 -0.0252 0.1892 -0.9804>}
 ring_02_r {parent=ring_01_r, loc=<-3.6174 -0.4647 0.1027>, rot=<0 0 0.2806 -0.9598>}
 ring_03_r {parent=ring_02_r, loc=<-2.4152 -0.114 -0.0199>, rot=<0 0 0.1189 -0.9929>}
 pinky_metacarpal_r {parent=hand_r, loc=<-1.7616 0.9753 -3.3814>, rot=<-0.0397 0.0814 -0.069 -0.9935>}
 pinky_01_r {parent=pinky_metacarpal_r, loc=<-7.1709 0.4831 0.1094>, rot=<-0.0654 -0.0035 0.2242 -0.9723>}
 pinky_02_r {parent=pinky_01_r, loc=<-2.8274 -0.0893 0.1043>, rot=<0.0037 0.0011 0.2854 -0.9584>}
 pinky_03_r {parent=pinky_02_r, loc=<-2.2046 -0.0307 0.0078>, rot=<-0.0002 0.0098 0.1445 -0.9895>}
 weapon_r {parent=hand_r, loc=<0.1974 3.4111 -0.3811>, rot=<0 0 0 1>}
 hand_attach_r {parent=hand_r, loc=<-8.5001 2.9999 0>, rot=<0 0 -0.7071 -0.7071>}
 deform_wrist_dn_base_r {parent=hand_r, loc=<-0.0004 2.2345 -0.0001>, rot=<0 0 0 -1>}
 deform_wrist_dn_r {parent=deform_wrist_dn_base_r, loc=<0 0 0>, rot=<0 0 0 1>}
 deform_wrist_up_base_r {parent=hand_r, loc=<0.0003 -2.2684 0.0001>, rot=<0 0 0 -1>}
 deform_wrist_up_r {parent=deform_wrist_up_base_r, loc=<0 0 0>, rot=<0 0 0 1>}
 lowerarm_twist_02_r {parent=lowerarm_r, loc=<-9.6199 0 0>, rot=<0 0 0 -1>}
 lowerarm_twist_01_r {parent=lowerarm_r, loc=<-17.8549 0 0>, rot=<0 0 0 -1>}
 deform_elbow_in_base_r {parent=lowerarm_r, loc=<-0.2111 -5.2591 -0.8275>, rot=<0 0 0 -1>}
 deform_elbow_in_r {parent=deform_elbow_in_base_r, loc=<-0.0001 0 0.0002>, rot=<0 0 0 1>}
 deform_elbow_out_base_r {parent=lowerarm_r, loc=<0 4 -0.0001>, rot=<0 0 0 -1>}
 deform_elbow_out_r {parent=deform_elbow_out_base_r, loc=<0 0 0>, rot=<0 0 0 1>}
 upperarm_twist_01_r {parent=upperarm_r, loc=<-0.0041 0 0>, rot=<0 0 0 -1>}
 deform_upperarm_bk_base_r {parent=upperarm_twist_01_r, loc=<-1.0814 7.2192 1.9455>, rot=<0 0 0 1>}
 deform_upperarm_bk_r {parent=deform_upperarm_bk_base_r, loc=<0 0 0>, rot=<0 0 0 1>}
 deform_upperarm_fr_base_r {parent=upperarm_twist_01_r, loc=<0.3689 -7.0506 2.7652>, rot=<0 0 0 1>}
 deform_upperarm_fr_r {parent=deform_upperarm_fr_base_r, loc=<0 0 0>, rot=<0 0 0 1>}
 deform_upperarm_up_base_r {parent=upperarm_twist_01_r, loc=<3.2214 1.1391 -3.3308>, rot=<0 0 0 1>}
 deform_upperarm_up_r {parent=deform_upperarm_up_base_r, loc=<0 0 0>, rot=<0 0 0 -1>}
 upperarm_twist_02_r {parent=upperarm_r, loc=<-13.7941 0 0>, rot=<0 0 0 -1>}
 dyn_collar_r {parent=spine_05, loc=<12.3251 7.6567 4.3854>, rot=<0.0324 0.3013 -0.3062 -0.9024>}
 dyn_collar_l {parent=spine_05, loc=<11.2796 8.2346 -3.5194>, rot=<-0.0764 -0.2877 -0.3515 -0.8876>}
 dyn_tie_base {parent=spine_05, loc=<1.9802 11.1548 0.1378>, rot=<0 0 0.9949 0.1007>}
 dyn_tie_01 {parent=dyn_tie_base, loc=<0 0 0>, rot=<0 0 0 1>}
 dyn_tie_02 {parent=dyn_tie_01, loc=<5.7451 0 0>, rot=<0 0 -0.0487 -0.9988>}
 dyn_tie_03 {parent=dyn_tie_02, loc=<7.2143 0 0>, rot=<0 0 -0.0688 -0.9976>}
 dyn_tie_04 {parent=dyn_tie_03, loc=<6.5544 0 0>, rot=<0 0 -0.041 -0.9992>}
 corrective_collar_r {parent=spine_05, loc=<13.1961 3.8112 6.5562>, rot=<0.0284 -0.7065 -0.0284 -0.7065>}
 corrective_collar_l {parent=spine_05, loc=<13.1961 3.8112 -5.9904>, rot=<0.0284 -0.7065 -0.0284 -0.7065>}
 deform_pec_aim_base_l {parent=spine_04, loc=<7.8728 12.6388 -2.4123>, rot=<0.0102 -0.4602 0.3979 -0.7936>}
 deform_pec_aim_l {parent=deform_pec_aim_base_l, loc=<0 0 0>, rot=<0 0 0 1>}
 deform_pec_base_l {parent=deform_pec_aim_l, loc=<10.766 0 0>, rot=<0 0 0 1>}
 deform_pec_l {parent=deform_pec_base_l, loc=<0.055 0 0>, rot=<0 0 0 1>}
 deform_pec_aim_base_r {parent=spine_04, loc=<7.8723 12.6388 2.4123>, rot=<-0.4602 -0.0102 0.7936 0.3979>}
 deform_pec_aim_r {parent=deform_pec_aim_base_r, loc=<0 0 0>, rot=<0 0 0 1>}
 deform_pec_base_r {parent=deform_pec_aim_r, loc=<-10.7662 -0.0006 -0.0007>, rot=<0 0 0 1>}
 deform_pec_r {parent=deform_pec_base_r, loc=<0 0 0>, rot=<0 0 0 1>}
 jacket_base {parent=spine_04, loc=<0 0 0>, rot=<0 0 0 1>}
 dyn_jacket_fr_r_base {parent=jacket_base, loc=<-2.6502 11.1718 10.4233>, rot=<-0.0191 0.5244 0.8513 -0.0024>}
 dyn_jacket_fr_r_01 {parent=dyn_jacket_fr_r_base, loc=<0 0 0>, rot=<0 0 0 1>}
 dyn_jacket_fr_r_02 {parent=dyn_jacket_fr_r_01, loc=<8.7223 0 0>, rot=<0 0 0 1>}
 dyn_jacket_s_r_base {parent=jacket_base, loc=<-2.555 -1.0271 17.0936>, rot=<-0.0154 0.0368 0.9991 -0.0115>}
 dyn_jacket_s_r_01 {parent=dyn_jacket_s_r_base, loc=<0 0 0>, rot=<0 0 0 1>}
 dyn_jacket_s_r_02 {parent=dyn_jacket_s_r_01, loc=<8.7223 0 0>, rot=<0 0 0 1>}
 dyn_jacket_b_r_base {parent=jacket_base, loc=<-2.0939 -11.3672 10.0515>, rot=<-0.0059 -0.5455 0.8379 -0.0183>}
 dyn_jacket_b_r_01 {parent=dyn_jacket_b_r_base, loc=<0 0 0>, rot=<0 0 0 1>}
 dyn_jacket_b_r_02 {parent=dyn_jacket_b_r_01, loc=<8.7223 0 0>, rot=<0 0 0 1>}
 dyn_jacket_b_l_base {parent=jacket_base, loc=<-1.5233 -11.9661 -8.5141>, rot=<0.0014 -0.8202 0.5718 -0.0192>}
 dyn_jacket_b_l_01 {parent=dyn_jacket_b_l_base, loc=<0 0 0>, rot=<0 0 0 1>}
 dyn_jacket_b_l_02 {parent=dyn_jacket_b_l_01, loc=<8.7223 0 0>, rot=<0 0 0 1>}
 dyn_jacket_s_l_base {parent=jacket_base, loc=<-1.5438 -0.7711 -16.8685>, rot=<0.012 -0.9998 0.0062 -0.015>}
 dyn_jacket_s_l_01 {parent=dyn_jacket_s_l_base, loc=<0 0 0>, rot=<0 0 0 1>}
 dyn_jacket_s_l_02 {parent=dyn_jacket_s_l_01, loc=<8.7223 0 0>, rot=<0 0 0 1>}
 dyn_jacket_fr_l_base {parent=jacket_base, loc=<-2.0349 11.1896 -10.1313>, rot=<0.0181 -0.8518 -0.5236 -0.0064>}
 dyn_jacket_fr_l_01 {parent=dyn_jacket_fr_l_base, loc=<0 0 0>, rot=<0 0 0 1>}
 dyn_jacket_fr_l_02 {parent=dyn_jacket_fr_l_01, loc=<8.7223 0 0>, rot=<0 0 0 1>}
 thigh_l {parent=pelvis, loc=<-2.5693 -0.0109 -7.7501>, rot=<0.0012 0.0572 0.0155 -0.9982>}
 calf_l {parent=thigh_l, loc=<-39.0484 0 0>, rot=<0.0006 -0.0161 -0.0391 -0.9991>}
 foot_l {parent=calf_l, loc=<-42.561 0.1077 0.4292>, rot=<0.0418 -0.0455 0.0343 -0.9975>}
 ball_l {parent=foot_l, loc=<-6.1599 13.8264 0.6806>, rot=<0.0001 -0.0001 -0.7063 -0.7079>}
 calf_twist_02_l {parent=calf_l, loc=<-18.8064 -0.0967 0.0397>, rot=<0.0237 0.0095 -0.0035 -0.9997>}
 calf_twist_01_l {parent=calf_l, loc=<-35.813 -0.0967 0.0397>, rot=<0.0237 0.0095 -0.0035 -0.9997>}
 dyn_pans_01_l {parent=calf_twist_01_l, loc=<2.8356 0.5761 -7.509>, rot=<0 0 0 1>}
 dyn_pans_02_l {parent=calf_twist_01_l, loc=<2.1492 0.3355 9.896>, rot=<0 0 0 1>}
 dyn_pans_03_l {parent=calf_twist_01_l, loc=<2.6992 -9.1816 -0.9171>, rot=<0 0 0 1>}
 dyn_pans_04_l {parent=calf_twist_01_l, loc=<3.3554 8.7986 0.1593>, rot=<0 0 0 1>}
 back_foot_corrective_l {parent=calf_twist_01_l, loc=<-6 -8 0>, rot=<0 0 0 1>}
 deform_knee_in_base_l {parent=calf_l, loc=<1.298 -5.2957 -0.0496>, rot=<0 0 0 -1>}
 deform_knee_in_l {parent=deform_knee_in_base_l, loc=<0 0 0>, rot=<0 0 0 1>}
 deform_knee_out_base_l {parent=calf_l, loc=<0 4 0>, rot=<0 0 0 -1>}
 deform_knee_out_l {parent=deform_knee_out_base_l, loc=<0 0 0>, rot=<0 0 0 1>}
 thigh_twist_01_l {parent=thigh_l, loc=<-21.6303 0 0>, rot=<0.0013 0 0 -1>}
 deform_groin_base_l {parent=thigh_l, loc=<-2.1908 8.1142 -1.7442>, rot=<0 0 0 -1>}
 deform_groin_l {parent=deform_groin_base_l, loc=<0 0 0>, rot=<0 0 0 1>}
 thigh_r {parent=pelvis, loc=<-2.5693 -0.0109 7.7501>, rot=<0.0572 -0.0012 0.9982 0.0155>}
 calf_r {parent=thigh_r, loc=<39.0484 0 0>, rot=<0.0006 -0.0161 -0.0391 -0.9991>}
 foot_r {parent=calf_r, loc=<42.561 -0.1077 -0.4292>, rot=<0.0418 -0.0455 0.0343 -0.9975>}
 ball_r {parent=foot_r, loc=<6.1599 -13.8264 -0.6806>, rot=<0.0001 -0.0001 -0.7063 -0.7079>}
 calf_twist_02_r {parent=calf_r, loc=<18.8064 0.0967 -0.0397>, rot=<0.0237 0.0095 -0.0035 -0.9997>}
 calf_twist_01_r {parent=calf_r, loc=<35.813 0.0967 -0.0397>, rot=<0.0237 0.0095 -0.0035 -0.9997>}
 dyn_pans_02_r {parent=calf_twist_01_r, loc=<-0.4261 -1.3214 7.1332>, rot=<0.1005 -0.0575 0.9933 0>}
 dyn_pans_04_r {parent=calf_twist_01_r, loc=<-3.3322 -8.8119 0.0698>, rot=<0.1005 -0.0575 0.9933 0>}
 dyn_pans_01_r {parent=calf_twist_01_r, loc=<-4.5764 0.4199 -9.6955>, rot=<0.1005 -0.0575 0.9933 0>}
 dyn_pans_03_r {parent=calf_twist_01_r, loc=<-3.0139 9.3616 -2.193>, rot=<0.1005 -0.0575 0.9933 0>}
 back_foot_corrective_r {parent=calf_twist_01_r, loc=<6 8 0>, rot=<0.1005 -0.0575 0.9933 0>}
 deform_knee_in_base_r {parent=calf_r, loc=<-1.298 5.2957 0.0496>, rot=<0 0 0 -1>}
 deform_knee_in_r {parent=deform_knee_in_base_r, loc=<0 0 0>, rot=<0 0 0 1>}
 deform_knee_out_base_r {parent=calf_r, loc=<0 -4 0>, rot=<0 0 0 -1>}
 deform_knee_out_r {parent=deform_knee_out_base_r, loc=<0 0 0>, rot=<0 0 0 1>}
 thigh_twist_01_r {parent=thigh_r, loc=<21.6303 0 0>, rot=<0.0013 0 0 -1>}
 deform_groin_base_r {parent=thigh_r, loc=<2.1908 -8.1142 1.7442>, rot=<0 0 0 -1>}
 deform_groin_r {parent=deform_groin_base_r, loc=<0 0 0>, rot=<0 0 0 1>}
 deform_glute_base_l {parent=pelvis, loc=<-8.8064 -8.8222 -8.4287>, rot=<0.0275 0.0501 -0.4472 -0.8926>}
 deform_glute_l {parent=deform_glute_base_l, loc=<0 0 0>, rot=<0 0 0 1>}
 deform_glute_base_r {parent=pelvis, loc=<-8.8062 -8.8222 8.4287>, rot=<0.0501 -0.0275 0.8926 -0.4472>}
 deform_glute_r {parent=deform_glute_base_r, loc=<0 0 0>, rot=<0 0 0 1>}
 dyn_thigh_trinket_base {parent=pelvis, loc=<9.8688 6.2514 -8.4964>, rot=<-0.0948 0.2368 0.9548 0.1525>}
 dyn_thigh_trinket_01 {parent=dyn_thigh_trinket_base, loc=<0 -0.0003 0>, rot=<0 0.0005 0 -1>}
 dyn_thigh_trinket_02 {parent=dyn_thigh_trinket_01, loc=<4.4772 0 0>, rot=<0 0 0 1>}
 dyn_thigh_trinket_03 {parent=dyn_thigh_trinket_02, loc=<1.2339 0 0>, rot=<0 0 0 1>}
 dyn_thigh_trinket_04 {parent=dyn_thigh_trinket_03, loc=<1.152 0 0>, rot=<0 0 0 1>}
 dyn_thigh_trinket_05 {parent=dyn_thigh_trinket_04, loc=<1.162 0 0>, rot=<0 0 0 1>}
 dyn_belt_b_base {parent=pelvis, loc=<11.2074 -10.5365 0.7563>, rot=<0.6059 0.14 0.7796 -0.0744>}
 dyn_belt_b_01 {parent=dyn_belt_b_base, loc=<0 0 0>, rot=<0 0 0 1>}
 dyn_belt_f_base {parent=pelvis, loc=<8.9411 9.6971 -0.6052>, rot=<-0.031 0.7739 0.0479 -0.6307>}
 dyn_belt_f_01 {parent=dyn_belt_f_base, loc=<0 0 0>, rot=<0 0 -0.0579 -0.9983>}
 dyn_katana_holder_base {parent=pelvis, loc=<5.8118 1.1293 15.7296>, rot=<-0.0179 -0.9982 -0.001 0.0568>}
 dyn_katana_holder_01 {parent=dyn_katana_holder_base, loc=<0 0 0>, rot=<0 0 0 1>}
 dyn_katana_holder_02 {parent=dyn_katana_holder_01, loc=<4.1522 0 0>, rot=<0.0004 -0.074 0.0037 -0.9973>}
 dyn_katana_holder_03 {parent=dyn_katana_holder_02, loc=<6.5706 0 0>, rot=<0 0 0 1>}
 dyn_katana_01_base {parent=dyn_katana_holder_03, loc=<1.1394 -0.6151 -1.2347>, rot=<-0.0563 0.0131 -0.0094 -0.9983>}
 dyn_katana_01 {parent=dyn_katana_01_base, loc=<0 0 0>, rot=<0 0 0 1>}
 dyn_katana_02_base {parent=dyn_katana_holder_03, loc=<7.2174 -0.4492 -1.1443>, rot=<-0.0561 0.0441 -0.02 -0.9972>}
 dyn_katana_02 {parent=dyn_katana_02_base, loc=<0 0 0>, rot=<0 0 0 1>}
 dyn_katana_02_trinket_01 {parent=dyn_katana_02, loc=<-4.2757 26.5479 -4.051>, rot=<0.9982 0.0127 0.0106 0.0576>}
 dyn_katana_02_trinket_02 {parent=dyn_katana_02_trinket_01, loc=<3.4743 0 0>, rot=<-0.0018 -0.0756 -0.0242 -0.9968>}
 dyn_katana_02_trinket_03 {parent=dyn_katana_02_trinket_02, loc=<1.5526 0 0>, rot=<0 0 0 1>}
 dyn_katana_02_trinket_04 {parent=dyn_katana_02_trinket_03, loc=<1.2473 0 0>, rot=<0 0 0 1>}
 dyn_katana_02_trinket_05 {parent=dyn_katana_02_trinket_04, loc=<1.003 0 0>, rot=<0 0 0 1>}
]
variables [
 Delta_Calf_L {type=double, default=0}
 Delta_Calf_R {type=double, default=0}
 Get pelvis from root by X {type=double, default=0}
 Is dbno {type=bool, default=False}
 Is_Dbno_alpha {type=double, default=0}
 RBAN is disable {type=bool, default=False}
 SPR_Thing_L {type=float, default=0}
 SPR_Thing_R {type=float, default=0}
 SPR_Thing_R_Side {type=double, default=0}
]
