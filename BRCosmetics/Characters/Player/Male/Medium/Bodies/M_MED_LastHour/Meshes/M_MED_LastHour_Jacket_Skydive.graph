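# Physics asset for the jacket and tie of the male medium LastHour player body, named as the skydive variant.
# Capsules on pelvis, spine and thighs form the body volume for front, side and back jacket chains and a tie chain.
# Chain bases join spine_05 with free rotation, while lower links use limited swing and mostly zero twist.
# An extra free-rotation constraint ties the left front jacket tip to the left side tip.
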PhysicsAsset /BRCosmetics/Characters/Player/Male/Medium/Bodies/M_MED_LastHour/Meshes/M_MED_LastHour_Jacket_Skydive
bodies {
 dyn_jacket_b_l_01 [{shapeType=Capsule, length=12.5452, radius=5, shapeName=dyn_jacket_b_l_01_capsule, center=<1.652 0 3>, rotation=<rot -0.0001 0 77.4198>}]
 dyn_jacket_b_l_02 [{
   shapeType=Capsule
   length=17.6401
   radius=5
   shapeName=dyn_jacket_b_l_02_capsule
   center=<4.5377 0 2.8517>
   rotation=<rot -0.0003 0.0001 77.4675>
  }]
 dyn_jacket_b_l_base [{shapeType=Capsule, length=0.505, radius=0.505, shapeName=dyn_jacket_b_l_base_capsule, rotation=<rot 0 -90 0>}]
 dyn_jacket_b_r_01 [{shapeType=Capsule, length=15.6886, radius=5, shapeName=dyn_jacket_b_r_01_capsule, center=<2.0644 0 3>, rotation=<rot -0.0001 0 98.076>}]
 dyn_jacket_b_r_02 [{
   shapeType=Capsule
   length=19.0407
   radius=5
   shapeName=dyn_jacket_b_r_02_capsule
   center=<4.1679 -0.0603 2.3261>
   rotation=<rot -0.0003 -0.0001 96.4969>
  }]
 dyn_jacket_b_r_base [{shapeType=Capsule, length=0.505, radius=0.505, shapeName=dyn_jacket_b_r_base_capsule, rotation=<rot 0 -90 0>}]
 dyn_jacket_fr_l_01 [{shapeType=Capsule, length=6, radius=4, shapeName=dyn_jacket_fr_l_01_capsule, center=<4.1503 0 1.8281>, rotation=<rot 0 0 90>}]
 dyn_jacket_fr_l_02 [{
   shapeType=Capsule
   length=15.0455
   radius=4
   shapeName=dyn_jacket_fr_l_02_capsule
   center=<5.0995 -3.3479 1>
   rotation=<rot 0.0005 0.0002 82.2051>
  }]
 dyn_jacket_fr_l_base [{shapeType=Capsule, length=0.505, radius=0.505, shapeName=dyn_jacket_fr_l_base_capsule, rotation=<rot 0 -90 0>}]
 dyn_jacket_fr_r_01 [{shapeType=Capsule, length=8, radius=5, shapeName=dyn_jacket_fr_r_01_capsule, center=<3.2334 0 3>, rotation=<rot 0 0 90>}]
 dyn_jacket_fr_r_02 [{shapeType=Capsule, length=8.6659, radius=3.3527, shapeName=dyn_jacket_fr_r_02_capsule, center=<4.4551 0 1.39>, rotation=<rot 0 0 90>}]
 dyn_jacket_fr_r_base [{shapeType=Capsule, length=0.505, radius=0.505, shapeName=dyn_jacket_fr_r_base_capsule, rotation=<rot 0 -90 0>}]
 dyn_jacket_s_l_01 [{shapeType=Capsule, length=8, radius=5, shapeName=dyn_jacket_s_l_01_capsule, center=<2.8062 0 3.2222>, rotation=<rot 0 0.0001 83.3944>}]
 dyn_jacket_s_l_02 [{shapeType=Capsule, length=17.7926, radius=3.4035, shapeName=dyn_jacket_s_l_02_capsule, center=<4.066 0 1.9002>, rotation=<rot 0 0 90>}]
 dyn_jacket_s_l_base [{shapeType=Capsule, length=0.505, radius=0.505, shapeName=dyn_jacket_s_l_base_capsule, rotation=<rot 0 -90 0>}]
 dyn_jacket_s_r_01 [{shapeType=Capsule, length=8, radius=5, shapeName=dyn_jacket_s_r_01_capsule, center=<2.1111 0 3>, rotation=<rot 0 0 90>}]
 dyn_jacket_s_r_02 [{shapeType=Capsule, length=14.266, radius=5, shapeName=dyn_jacket_s_r_02_capsule, center=<4.2558 0 4.51>, rotation=<rot 0 0 90>}]
 dyn_jacket_s_r_base [{shapeType=Capsule, length=0.505, radius=0.505, shapeName=dyn_jacket_s_r_base_capsule, rotation=<rot 0 -90 0>}]
 dyn_tie_01 [{
   shapeType=Capsule
   length=2.1491
   radius=1.4934
   shapeName=dyn_tie_01_capsule
   center=<3.637 -1.2 0.1056>
   rotation=<rot 0 -90.1353 -87.3035>
  }]
 dyn_tie_02 [{
   shapeType=Capsule
   length=2.5217
   radius=1.9591
   shapeName=dyn_tie_02_capsule
   center=<4.5831 -1.3 -0.2081>
   rotation=<rot 0 92.7138 -85.0185>
  }]
 dyn_tie_03 [{shapeType=Capsule, length=6.5017, radius=2.675, shapeName=dyn_tie_03_capsule, center=<8.5589 -2 -0.0974>, rotation=<rot 0 -86.7668 -89.4932>}]
 dyn_tie_04 [{shapeType=Capsule, length=0.505, radius=0.505, shapeName=dyn_tie_04_capsule, center=<0 -0.2887 0>}]
 dyn_tie_base [{shapeType=Capsule, length=0.505, radius=0.505, shapeName=dyn_tie_base_capsule, rotation=<rot 0 -90 0>}]
 pelvis [{shapeType=Capsule, length=17.9968, radius=10.5982, shapeName=pelvis_capsule, center=<4.9345 -1.1143 0>}]
 spine_01 [{
   shapeType=Capsule
   length=9.3457
   radius=10.1991
   shapeName=spine_01_capsule
   center=<5.4246 -0.7346 -0.0262>
   rotation=<rot 0.1148 -90 0.2879>
  }]
 spine_02 [{shapeType=Capsule, length=6.0535, radius=10.7699, shapeName=spine_02_capsule, center=<0 0.6201 0>}]
 spine_04 [{
   shapeType=Capsule
   length=11.7819
   radius=12.0247
   shapeName=spine_04_capsule
   center=<0.9154 0.3393 -0.0174>
   rotation=<rot 88.9094 -0.0001 0>
  }]
 spine_05 [{shapeType=Capsule, length=7, radius=10, shapeName=spine_05_capsule}]
 thigh_l [{
   shapeType=Capsule
   length=32.1496
   radius=10.1495
   shapeName=thigh_l_capsule
   center=<-14.3065 -0.1296 1.0883>
   rotation=<rot -0.5739 -90.4394 -89.148>
  }]
 thigh_r [{
   shapeType=Capsule
   length=32.1496
   radius=10.1495
   shapeName=thigh_l_capsule
   center=<14.2907 -0.334 -1.0927>
   rotation=<rot -0.5524 -91.9423 90.8669>
  }]
}
constraints [
 {
  bone1Name=spine_05
  bone2Name=spine_04
  swing1LimitDegrees=45
  swing1Motion=Free
  swing2LimitDegrees=45
  swing2Motion=Free
  twistLimitDegrees=45
  twistMotion=Free
 }
 {
  bone1Name=spine_04
  bone2Name=spine_02
  swing1LimitDegrees=45
  swing1Motion=Free
  swing2LimitDegrees=45
  swing2Motion=Free
  twistLimitDegrees=45
  twistMotion=Free
 }
 {
  bone1Name=dyn_jacket_b_l_base
  bone2Name=spine_05
  swing1LimitDegrees=45
  swing1Motion=Free
  swing2LimitDegrees=45
  swing2Motion=Free
  twistLimitDegrees=45
  twistMotion=Free
 }
 {bone1Name=dyn_jacket_b_l_01, bone2Name=dyn_jacket_b_l_base, swing1LimitDegrees=5, swing2LimitDegrees=25, twistLimitDegrees=0}
 {bone1Name=dyn_jacket_b_l_02, bone2Name=dyn_jacket_b_l_01, swing1LimitDegrees=5, swing2LimitDegrees=25, twistLimitDegrees=0}
 {
  bone1Name=dyn_jacket_b_r_base
  bone2Name=spine_05
  swing1LimitDegrees=45
  swing1Motion=Free
  swing2LimitDegrees=45
  swing2Motion=Free
  twistLimitDegrees=45
  twistMotion=Free
 }
 {bone1Name=dyn_jacket_b_r_01, bone2Name=dyn_jacket_b_r_base, swing1LimitDegrees=5, swing2LimitDegrees=25, twistLimitDegrees=0}
 {bone1Name=dyn_jacket_b_r_02, bone2Name=dyn_jacket_b_r_01, swing1LimitDegrees=5, swing2LimitDegrees=25, twistLimitDegrees=0}
 {
  bone1Name=dyn_jacket_fr_l_base
  bone2Name=spine_05
  swing1LimitDegrees=45
  swing1Motion=Free
  swing2LimitDegrees=45
  swing2Motion=Free
  twistLimitDegrees=45
  twistMotion=Free
 }
 {
  bone1Name=dyn_jacket_fr_r_base
  bone2Name=spine_05
  swing1LimitDegrees=45
  swing1Motion=Free
  swing2LimitDegrees=45
  swing2Motion=Free
  twistLimitDegrees=45
  twistMotion=Free
 }
 {bone1Name=dyn_jacket_fr_r_01, bone2Name=dyn_jacket_fr_r_base, swing1LimitDegrees=5, swing2LimitDegrees=45, twistLimitDegrees=0}
 {bone1Name=dyn_jacket_fr_r_02, bone2Name=dyn_jacket_fr_r_01, swing1LimitDegrees=15, swing2LimitDegrees=25, twistLimitDegrees=0}
 {
  bone1Name=dyn_jacket_s_l_base
  bone2Name=spine_05
  swing1LimitDegrees=45
  swing1Motion=Free
  swing2LimitDegrees=45
  swing2Motion=Free
  twistLimitDegrees=45
  twistMotion=Free
 }
 {bone1Name=dyn_jacket_s_l_01, bone2Name=dyn_jacket_s_l_base, swing1LimitDegrees=25, swing2LimitDegrees=25, twistLimitDegrees=0}
 {
  bone1Name=dyn_jacket_s_r_base
  bone2Name=spine_05
  swing1LimitDegrees=45
  swing1Motion=Free
  swing2LimitDegrees=45
  swing2Motion=Free
  twistLimitDegrees=45
  twistMotion=Free
 }
 {bone1Name=dyn_jacket_s_r_01, bone2Name=dyn_jacket_s_r_base, swing1LimitDegrees=5, swing2LimitDegrees=25, twistLimitDegrees=0}
 {bone1Name=dyn_jacket_s_r_02, bone2Name=dyn_jacket_s_r_01, swing1LimitDegrees=5, swing2LimitDegrees=25, twistLimitDegrees=0}
 {bone1Name=dyn_jacket_fr_l_02, bone2Name=dyn_jacket_fr_l_01, swing1LimitDegrees=27.28, swing2LimitDegrees=25, twistLimitDegrees=25}
 {bone1Name=dyn_jacket_fr_l_01, bone2Name=dyn_jacket_fr_l_base, swing1LimitDegrees=10, swing2LimitDegrees=45, twistLimitDegrees=10}
 {bone1Name=dyn_jacket_s_l_02, bone2Name=dyn_jacket_s_l_01, swing1LimitDegrees=24.32, swing2LimitDegrees=45, twistLimitDegrees=0}
 {
  bone1Name=dyn_jacket_fr_l_02
  bone2Name=dyn_jacket_s_l_02
  swing1LimitDegrees=45
  swing1Motion=Free
  swing2LimitDegrees=45
  swing2Motion=Free
  twistLimitDegrees=45
  twistMotion=Free
 }
 {
  bone1Name=dyn_tie_base
  bone2Name=spine_05
  swing1LimitDegrees=45
  swing1Motion=Free
  swing2LimitDegrees=45
  swing2Motion=Free
  twistLimitDegrees=45
  twistMotion=Free
 }
 {bone1Name=dyn_tie_01, bone2Name=dyn_tie_base, swing1LimitDegrees=25, swing2LimitDegrees=1, twistLimitDegrees=0}
 {bone1Name=dyn_tie_02, bone2Name=dyn_tie_01, swing1LimitDegrees=25, swing2LimitDegrees=1, twistLimitDegrees=0}
 {bone1Name=dyn_tie_03, bone2Name=dyn_tie_02, swing1LimitDegrees=25, swing2LimitDegrees=2, twistLimitDegrees=0}
 {bone1Name=dyn_tie_04, bone2Name=dyn_tie_03, swing1LimitDegrees=25, swing2LimitDegrees=2, twistLimitDegrees=0}
 {
  bone1Name=spine_02
  bone2Name=spine_01
  swing1LimitDegrees=45
  swing1Motion=Free
  swing2LimitDegrees=45
  swing2Motion=Free
  twistLimitDegrees=45
  twistMotion=Free
 }
 {
  bone1Name=thigh_l
  bone2Name=pelvis
  swing1LimitDegrees=45
  swing1Motion=Free
  swing2LimitDegrees=45
  swing2Motion=Free
  twistLimitDegrees=45
  twistMotion=Free
 }
 {
  bone1Name=thigh_r
  bone2Name=pelvis
  swing1LimitDegrees=45
  swing1Motion=Free
  swing2LimitDegrees=45
  swing2Motion=Free
  twistLimitDegrees=45
  twistMotion=Free
 }
 {
  bone1Name=spine_01
  bone2Name=pelvis
  swing1LimitDegrees=45
  swing1Motion=Free
  swing2LimitDegrees=45
  swing2Motion=Free
  twistLimitDegrees=45
  twistMotion=Free
 }
]
