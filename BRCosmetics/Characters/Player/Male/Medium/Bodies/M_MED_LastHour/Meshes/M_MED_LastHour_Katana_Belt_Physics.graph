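# Physics asset for the katana belt of the male medium LastHour character cosmetic, covering two katanas, their holder, belt pieces and a thigh trinket.
# The katanas hang from a three-link holder chain and the trinket forms a five-link chain, both angle-limited from base bodies constrained to the pelvis.
# Front and back belt bodies are rotation-locked to the pelvis.
# Capsules on pelvis, thighs and right calf, plus a thin root box, supply body and base collision shapes.
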PhysicsAsset /BRCosmetics/Characters/Player/Male/Medium/Bodies/M_MED_LastHour/Meshes/M_MED_LastHour_Katana_Belt_Physics
bodies {
 calf_r [{
   shapeType=Capsule
   length=32.1678
   radius=9.1959
   shapeName=calf_r_capsule
   center=<20.1803 -1.1354 -1.2204>
   rotation=<rot 0 -86.1226 -89.3114>
  }]
 dyn_belt_b_01 [{shapeType=Capsule, length=2, radius=1, shapeName=dyn_belt_b_01_capsule}]
 dyn_belt_f_01 [{shapeType=Capsule, length=3, radius=2.5, shapeName=dyn_belt_f_01_capsule, rotation=<rot 90 0 0>}]
 dyn_katana_01 [{
   shapeType=Capsule
   length=74.8116
   radius=1.9865
   shapeName=dyn_katana_01_capsule
   center=<2.8267 -4.9194 -0.8331>
   rotation=<rot 0 24.8782 -80.1993>
  }]
 dyn_katana_02 [{
   shapeType=Capsule
   length=95.6552
   radius=2.183
   shapeName=dyn_katana_02_capsule
   center=<3.4766 -13.8455 -0.5248>
   rotation=<rot 0 12.9382 -85.3087>
  }]
 dyn_katana_holder_01 [{shapeType=Capsule, length=7, radius=5, shapeName=dyn_katana_holder_01_capsule, center=<0 0 -3.7994>, rotation=<rot 0 0 90>}]
 dyn_katana_holder_02 [{shapeType=Capsule, length=7, radius=5, shapeName=dyn_katana_holder_02_capsule, center=<0 0 -3.6603>, rotation=<rot 0 0 90>}]
 dyn_katana_holder_03 [{
   shapeType=Capsule
   length=4.5872
   radius=3.7222
   shapeName=dyn_katana_holder_03_capsule
   center=<3.7085 -0.5519 -2.1799>
   rotation=<rot 0 167.7931 -83.595>
  }]
 dyn_katana_holder_base [{shapeType=Capsule, length=0.505, radius=0.505, shapeName=dyn_katana_holder_base_capsule, rotation=<rot 0 -90 0>}]
 dyn_thigh_trinket_01 [{shapeType=Capsule, length=5, radius=1.5, shapeName=dyn_thigh_trinket_01_capsule, center=<0 -1 0>}]
 dyn_thigh_trinket_02 [{shapeType=Capsule, length=1, radius=1, shapeName=dyn_thigh_trinket_02_capsule, center=<0.1225 -0.5 0>}]
 dyn_thigh_trinket_03 [{shapeType=Capsule, length=1, radius=1, shapeName=dyn_thigh_trinket_03_capsule, center=<0.2296 -0.5 0>}]
 dyn_thigh_trinket_04 [{shapeType=Capsule, length=1, radius=1, shapeName=dyn_thigh_trinket_04_capsule, center=<0.5465 -0.5 0>}]
 dyn_thigh_trinket_05 [
  {
   shapeType=Capsule
   length=5.5568
   radius=1.5
   shapeName=dyn_thigh_trinket_05_capsule
   center=<2.4871 -1 0.4625>
   rotation=<rot 30.1798 -0.0001 0>
  }
  {
   shapeType=Capsule
   length=5.5568
   radius=1.5
   shapeName=dyn_thigh_trinket_05_capsule
   center=<2.4871 -1 -0.3864>
   rotation=<rot -34.1147 -0.0001 -0.0001>
  }
 ]
 dyn_thigh_trinket_base [{shapeType=Capsule, length=0.505, radius=0.505, shapeName=dyn_thigh_trinket_base_capsule, rotation=<rot 0 -90 0>}]
 pelvis [{
   shapeType=Capsule
   length=4.3477
   radius=12.6958
   shapeName=pelvis_capsule
   center=<1.0713 -2.9578 -0.0422>
   rotation=<rot 2.255 179.9999 -180>
  }]
 root [{shapeType=Box, extentX=141.01, extentY=123.01, extentZ=3.01, shapeName=root_box, center=<-51.2009 -13.209 0>, rotation=<rot 0 -90 0>}]
 thigh_l [{
   shapeType=Capsule
   length=30.7011
   radius=10.0867
   shapeName=thigh_r_capsule
   center=<-15.2945 0.0873 1.3303>
   rotation=<rot 0.0936 -93.3136 88.1167>
  }]
 thigh_r [{
   shapeType=Capsule
   length=32.3586
   radius=10.0867
   shapeName=thigh_r_capsule
   center=<21.1696 -0.7414 -1.6322>
   rotation=<rot 0.1001 -93.3696 -88.3908>
  }]
}
constraints [
 {
  bone1Name=dyn_belt_f_01
  bone2Name=pelvis
  swing1LimitDegrees=45
  swing1Motion=Locked
  swing2LimitDegrees=45
  swing2Motion=Locked
  twistLimitDegrees=45
  twistMotion=Locked
 }
 {
  bone1Name=dyn_belt_b_01
  bone2Name=pelvis
  swing1LimitDegrees=45
  swing1Motion=Locked
  swing2LimitDegrees=45
  swing2Motion=Locked
  twistLimitDegrees=45
  twistMotion=Locked
 }
 {bone1Name=dyn_katana_02, bone2Name=dyn_katana_holder_03, swing1LimitDegrees=25, swing2LimitDegrees=5, twistLimitDegrees=5}
 {bone1Name=dyn_katana_01, bone2Name=dyn_katana_holder_03, swing1LimitDegrees=25, swing2LimitDegrees=5, twistLimitDegrees=0}
 {
  bone1Name=thigh_r
  bone2Name=pelvis
  swing1LimitDegrees=45
  swing1Motion=Free
  swing2LimitDegrees=45
  swing2Motion=Free
  twistLimitDegrees=45
  twistMotion=Free
 }
 {
  bone1Name=calf_r
  bone2Name=thigh_r
  swing1LimitDegrees=45
  swing1Motion=Free
  swing2LimitDegrees=45
  swing2Motion=Free
  twistLimitDegrees=45
  twistMotion=Free
 }
 {
  bone1Name=pelvis
  bone2Name=root
  swing1LimitDegrees=45
  swing1Motion=Free
  swing2LimitDegrees=45
  swing2Motion=Free
  twistLimitDegrees=45
  twistMotion=Free
 }
 {
  bone1Name=dyn_thigh_trinket_base
  bone2Name=pelvis
  swing1LimitDegrees=45
  swing1Motion=Free
  swing2LimitDegrees=45
  swing2Motion=Free
  twistLimitDegrees=45
  twistMotion=Free
 }
 {bone1Name=dyn_thigh_trinket_01, bone2Name=dyn_thigh_trinket_base, swing1LimitDegrees=15, swing2LimitDegrees=15, twistLimitDegrees=0}
 {bone1Name=dyn_thigh_trinket_02, bone2Name=dyn_thigh_trinket_01, swing1LimitDegrees=29.4, swing2LimitDegrees=35.16, twistLimitDegrees=25}
 {bone1Name=dyn_thigh_trinket_03, bone2Name=dyn_thigh_trinket_02, swing1LimitDegrees=15, swing2LimitDegrees=15, twistLimitDegrees=20}
 {bone1Name=dyn_thigh_trinket_04, bone2Name=dyn_thigh_trinket_03, swing1LimitDegrees=25.08, swing2LimitDegrees=30.84, twistLimitDegrees=40.16}
 {bone1Name=dyn_thigh_trinket_05, bone2Name=dyn_thigh_trinket_04, swing1LimitDegrees=27.96, swing2LimitDegrees=33.72, twistLimitDegrees=20}
 {
  bone1Name=thigh_l
  bone2Name=pelvis
  swing1LimitDegrees=45
  swing1Motion=Free
  swing2LimitDegrees=45
  swing2Motion=Free
  twistLimitDegrees=45
  twistMotion=Free
 }
 {
  bone1Name=dyn_katana_holder_base
  bone2Name=pelvis
  swing1LimitDegrees=45
  swing1Motion=Free
  swing2LimitDegrees=45
  swing2Motion=Free
  twistLimitDegrees=45
  twistMotion=Free
 }
 {bone1Name=dyn_katana_holder_01, bone2Name=dyn_katana_holder_base, swing1LimitDegrees=0, swing2LimitDegrees=0, twistLimitDegrees=0}
 {bone1Name=dyn_katana_holder_02, bone2Name=dyn_katana_holder_01, swing1LimitDegrees=5, swing2LimitDegrees=36.36, twistLimitDegrees=8.64}
 {bone1Name=dyn_katana_holder_03, bone2Name=dyn_katana_holder_02, swing1LimitDegrees=5, swing2LimitDegrees=45, twistLimitDegrees=0}
]
