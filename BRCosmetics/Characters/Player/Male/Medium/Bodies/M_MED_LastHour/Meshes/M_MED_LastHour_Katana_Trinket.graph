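# Physics asset for the dangling katana trinket on the medium male Last Hour player cosmetic body.
# Capsule bodies form a chain from the pelvis through the second katana bone to five trinket bones, with right thigh and calf capsules covering the leg.
# The first trinket joint is set to swing on one axis, the later links share tighter cone and twist limits, and the leg joints rotate freely.
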PhysicsAsset /BRCosmetics/Characters/Player/Male/Medium/Bodies/M_MED_LastHour/Meshes/M_MED_LastHour_Katana_Trinket
bodies {
 calf_r [{
   shapeType=Capsule
   length=32.1678
   radius=9.1959
   shapeName=calf_r_capsule
   center=<20.1803 -1.1354 -1.2204>
   rotation=<rot 0 -86.1226 -89.3114>
  }]
 dyn_katana_02 [{shapeType=Capsule, length=11.9409, radius=2.183, shapeName=dyn_katana_02_capsule, rotation=<rot 0 13.9833 90>}]
 dyn_katana_02_trinket_01 [{
   shapeType=Capsule
   length=1.8997
   radius=1.6497
   shapeName=dyn_katana_02_trinket_01_capsule
   center=<1.4333 -0.0104 -0.0538>
   rotation=<rot 0 90.1916 -86.6946>
  }]
 dyn_katana_02_trinket_02 [{
   shapeType=Capsule
   length=0.8087
   radius=0.5546
   shapeName=dyn_katana_02_trinket_02_capsule
   center=<0.7509 0.0326 -0.0115>
   rotation=<rot 0 -91.4195 -87.2411>
  }]
 dyn_katana_02_trinket_03 [{
   shapeType=Capsule
   length=0.505
   radius=0.505
   shapeName=dyn_katana_02_trinket_03_capsule
   center=<0.4804 0.0346 0.0738>
   rotation=<rot 0 -102.5375 -58.2011>
  }]
 dyn_katana_02_trinket_04 [{
   shapeType=Capsule
   length=0.505
   radius=0.505
   shapeName=dyn_katana_02_trinket_04_capsule
   center=<0.4826 -0.0184 -0.0023>
   rotation=<rot 0 90.7458 -88.6674>
  }]
 dyn_katana_02_trinket_05 [{
   shapeType=Capsule
   length=3.2873
   radius=1.0401
   shapeName=dyn_katana_02_trinket_05_capsule
   center=<1.0586 -0.0726 -0.0055>
   rotation=<rot 0 -93.7408 -89.8076>
  }]
 pelvis [{
   shapeType=Capsule
   length=4.3477
   radius=12.6958
   shapeName=pelvis_capsule
   center=<1.0713 -2.9578 -0.0422>
   rotation=<rot 2.255 179.9999 -180>
  }]
 thigh_r [{
   shapeType=Capsule
   length=32.3586
   radius=10.0867
   shapeName=thigh_r_capsule
   center=<21.1696 -0.7414 -1.6322>
   rotation=<rot 0.1001 -93.3696 -88.3908>
  }]
}
constraints [
 {bone1Name=dyn_katana_02, bone2Name=pelvis, swing1LimitDegrees=25, swing2LimitDegrees=5, twistLimitDegrees=5}
 {
  bone1Name=thigh_r
  bone2Name=pelvis
  swing1LimitDegrees=45
  swing1Motion=Free
  swing2LimitDegrees=45
  swing2Motion=Free
  twistLimitDegrees=45
  twistMotion=Free
 }
 {
  bone1Name=calf_r
  bone2Name=thigh_r
  swing1LimitDegrees=45
  swing1Motion=Free
  swing2LimitDegrees=45
  swing2Motion=Free
  twistLimitDegrees=45
  twistMotion=Free
 }
 {bone1Name=dyn_katana_02_trinket_01, bone2Name=dyn_katana_02, swing1LimitDegrees=71.92, swing2LimitDegrees=0, twistLimitDegrees=0}
 {bone1Name=dyn_katana_02_trinket_02, bone2Name=dyn_katana_02_trinket_01, swing1LimitDegrees=25, swing2LimitDegrees=25, twistLimitDegrees=10}
 {bone1Name=dyn_katana_02_trinket_03, bone2Name=dyn_katana_02_trinket_02, swing1LimitDegrees=25, swing2LimitDegrees=25, twistLimitDegrees=10}
 {bone1Name=dyn_katana_02_trinket_04, bone2Name=dyn_katana_02_trinket_03, swing1LimitDegrees=25, swing2LimitDegrees=25, twistLimitDegrees=10}
 {bone1Name=dyn_katana_02_trinket_05, bone2Name=dyn_katana_02_trinket_04, swing1LimitDegrees=25, swing2LimitDegrees=25, twistLimitDegrees=10}
]
