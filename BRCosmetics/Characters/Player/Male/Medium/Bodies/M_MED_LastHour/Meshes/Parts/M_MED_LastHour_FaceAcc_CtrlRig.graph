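# Control Rig for the M_MED_LastHour face accessory that drives corrective hair bones from head rotation.
# It reads the head rotation relative to spine_05, converts it to ZYX Euler, and remaps a negative Z-angle range to a blend weight.
# That weight feeds a Modify Transforms step that applies preset offsets to the corrective_hair bones.
# The dyn_hair bones are in the hierarchy, but no visible node drives them.
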ControlRigBlueprint /BRCosmetics/Characters/Player/Male/Medium/Bodies/M_MED_LastHour/Meshes/Parts/M_MED_LastHour_FaceAcc_CtrlRig
graphs [
 RigVMFunctionLibrary {links [], nodes []}
 RigVMModel {nodes [
   RigUnit_BeginExecution = &RigUnit_BeginExecution()
   RigVMFunction_MathQuaternionToEuler = &RigVMFunction_MathQuaternionToEuler(
    RotationOrder=ZYX
    Value=&RigUnit_GetRelativeTransformForItem(Child=(Type=Bone,Name="head"), Parent=(Type=Bone,Name="spine_05"), bChildInitial=False, bParentInitial=False).RelativeTransform.Rotation
   )
   Remap = &RigVMFunction_MathDoubleRemap(SourceMaximum=-60, SourceMinimum=-40, TargetMaximum=1, TargetMinimum=0, Value=@RigVMFunction_MathQuaternionToEuler.Result.Z)
   &RigUnit_ModifyTransforms(
    ItemToModify="((Item=(Type=Bone,Name=\"corrective_hair_01\"),Transform=(Rotation=(X=0,Y=0,Z=0,W=1),Translation=(X=0,Y=0,Z=2),Scale3D=(X=1,Y=1,Z=1))),(Item=(Type=Bone,Name=\"corrective_hair_02\"),Transform=(Rotation=(X=0,Y=0,Z=0,W=1),Translation=(X=0,Y=0,Z=0),Scale3D=(X=1,Y= ...(truncated)"
    WeightMinimum=0
    Weight=@Remap.Result
    ExecutePin=@RigUnit_BeginExecution.ExecutePin
   )
  ]}
]
hierarchy [
 root {loc=<0 0 0>, rot=<0 0 0 1>}
 pelvis {parent=root, loc=<0 -0.7071 93.1191>, rot=<0.0127 0.707 -0.0127 -0.707>}
 spine_01 {parent=pelvis, loc=<6.7494 -0.6388 0>, rot=<0 0 0.034 -0.9994>}
 spine_02 {parent=spine_01, loc=<7.252 -0.035 0>, rot=<0 0 0.0238 -0.9997>}
 spine_03 {parent=spine_02, loc=<6.0106 0 0>, rot=<0 0 0.0149 -0.9999>}
 spine_04 {parent=spine_03, loc=<8.1023 0 0>, rot=<0 0 -0.0132 -0.9999>}
 spine_05 {parent=spine_04, loc=<10.0142 0 0>, rot=<0 0 -0.0015 -1>}
 neck_01 {parent=spine_05, loc=<11.3138 0.1529 0>, rot=<0 0 -0.2003 -0.9797>}
 neck_02 {parent=neck_01, loc=<4.5627 0 0>, rot=<0 0 0.0504 -0.9987>}
 head {parent=neck_02, loc=<5.3914 0 0>, rot=<0 0 0.115 -0.9934>}
 dyn_hair_01 {parent=head, loc=<5.4242 2.8809 -7.8331>, rot=<0.0015 -0.9974 -0.0499 0.0519>}
 dyn_hair_02 {parent=head, loc=<7.8263 -5.9789 -5.5283>, rot=<0.4039 -0.5804 0.3991 -0.5837>}
 dyn_hair_03 {parent=head, loc=<12.5686 4.6253 -6.1291>, rot=<0.3217 -0.6166 0.2726 -0.6648>}
 dyn_hair_04 {parent=head, loc=<7.1101 -7.4109 4.2174>, rot=<0.5387 -0.7023 -0.2425 -0.3973>}
 dyn_hair_05 {parent=head, loc=<5.4803 6.6329 7.6885>, rot=<-0.0964 0.1839 0.9781 -0.012>}
 dyn_hair_06 {parent=dyn_hair_05, loc=<3.3415 0 0>, rot=<0 0 0 1>}
 dyn_hair_07 {parent=head, loc=<9.0106 1.4699 8.1483>, rot=<0.763 0.0013 0.5904 -0.2633>}
 dyn_hair_08 {parent=head, loc=<7.42 11.3001 0.5028>, rot=<-0.0758 0.165 0.9827 0.0359>}
 dyn_hair_09 {parent=head, loc=<15.1061 -2.1405 2.2436>, rot=<0.7967 0.2057 0.2362 -0.5169>}
 corrective_hair_01 {parent=head, loc=<0.2321 -5.7796 -0.5244>, rot=<0.0029 -0.7071 -0.0029 -0.7071>}
 corrective_hair_02 {parent=head, loc=<0.2082 -2.883 -5.2101>, rot=<0.0029 -0.7071 -0.0029 -0.7071>}
 corrective_hair_03 {parent=head, loc=<0.2045 -2.4346 5.7483>, rot=<0.0029 -0.7071 -0.0029 -0.7071>}
]
